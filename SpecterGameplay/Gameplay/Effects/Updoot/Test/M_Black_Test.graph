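# Unlit test material in the Updoot effects folder, named as a flat black surface.
# Its graph holds a single Constant3Vector left at its default zero value, with no textures, parameters or other logic.
# Usage is enabled for skeletal meshes, mesh deformers and Niagara mesh particles, and automatic usage setting in the editor is turned off.
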Material /SpecterGameplay/Gameplay/Effects/Updoot/Test/M_Black_Test
graph {
 properties {
  bAutomaticallySetUsageInEditor=false
  bUsedWithMeshDeformer=true
  bUsedWithNiagaraMeshParticles=true
  bUsedWithSkeletalMesh=true
  shadingModel=MSM_Unlit
  usedShadingModels=Unlit
 }
 nodes [Constant3Vector()]
}
