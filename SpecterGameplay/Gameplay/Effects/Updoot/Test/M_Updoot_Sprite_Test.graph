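# Test translucent subsurface material for the Updoot sprite effect mesh, styled as a glowing, bubbly body with an arrow logo.
# Emissive layers time-of-day-scaled color blends, noise-warped inner glow, sun-reflection scattering and a bump-offset cropped arrow, then applies inverse eye adaptation.
# Fresnel outer glow, panning bubble normals and object-anchored, depth- and FOV-scaled screen UVs are built as named reroutes.
# Vertex colors tint emissive and subsurface green; MF_Sprite_Updoot_WPO drives vertex offset.
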Material /SpecterGameplay/Gameplay/Effects/Updoot/Test/M_Updoot_Sprite_Test
graph {
 properties {
  bAutomaticallySetUsageInEditor=false
  bUseMaterialAttributes=true
  bUsedWithMeshDeformer=true
  bUsedWithNiagaraMeshParticles=true
  bUsedWithSkeletalMesh=true
  blendMode=BLEND_Translucent
  previewMesh=&/SM_Sprite_Updoot_FXProtoTest
  shadingModel=MSM_Subsurface
  usedShadingModels=Subsurface
 }
 nodes [
  n1 = ObjectPositionWS()
  n2 = TransformPosition(None=@n1) {transformSourceType=TRANSFORMPOSSOURCE_World, transformType=TRANSFORMPOSSOURCE_View}
  ComponentMask(None=@n2) {b=true, g=true}
  Constant2Vector() {g=-0.1, r=0.315}
  Constant() {r=2}
  Divide(A=TransformPosition(None=WorldPosition().XYZ) {transformSourceType=TRANSFORMPOSSOURCE_World, transformType=TRANSFORMPOSSOURCE_View}) {constB=2000}
  n3 = Constant() {r=1}
  n4 = Multiply(
   A=MaterialFunctionCall(Value_Morning=@n3, Value_Day=@n3, Value_Evening=Constant() {r=0.5}, Value_Night=Constant()) {materialFunction=&Scalar_Time_of_Day}.Result
   B=VectorParameter() {parameterName=BaseColorA_Scattered, defaultValue=<rgba 0.1343 2 0.0158 1>}.RGB
  )
  n5 = NamedRerouteUsage()
  n6 = Constant3Vector() {constant=<rgba 0.3613 0.9823 1 1>}
  n7 = VectorParameter() {parameterName=BaseColorC, defaultValue=<rgba 0.3372 2 0.7559 1>}
  n8 = Constant() {r=1}
  n9 = Multiply(
   A=MaterialFunctionCall(Value_Morning=@n8, Value_Day=@n8, Value_Evening=Constant() {r=0.5}, Value_Night=Constant()) {materialFunction=&Scalar_Time_of_Day}.Result
   B=VectorParameter() {parameterName=BaseColorScattered, defaultValue=<rgba 0.0908 0.5841 0.1746 1>}.RGB
  )
  n10 = NamedRerouteUsage()
  n11 = Divide(A=ComponentMask(None=NamedRerouteUsage()) {b=true}) {constB=35}
  n12 = Constant() {r=0.35}
  n13 = LinearInterpolate(
   A=VectorParameter() {parameterName=BaseColorB, defaultValue=<rgba 0.2789 0.4852 0.956 1>}.RGB
   B=LinearInterpolate(A=LinearInterpolate(A=@n6, B=@n4, Alpha=@n5), B=LinearInterpolate(A=@n7.RGB, B=@n9, Alpha=@n10), Alpha=Step(Y=@n11, X=@n12) {constX=0.4})
   Alpha=Distance(A=@n11, B=@n12)
  )
  n14 = VectorParameter() {parameterName=InnerGlowColor, defaultValue=<rgba 0 0.9767 0.9911 1>}
  n15 = Constant()
  n16 = ScalarParameter() {defaultValue=0.3, parameterName=InnerGlowNoiseSpeed}
  n17 = TextureSample(UVs=Panner(Coordinate=NamedRerouteUsage(), Speed=AppendVector(A=@n15, B=@n16)) {speedY=0.2}) {texture=/Game/Environments/Helios/Props/Quest/Quest_GoldenFleeceStatuette_A/Textures/T_GoldenFleeceStatuette_CellNoise}
  n18 = ComponentMask(None=Multiply(A=ConstantBiasScale(None=@n17.RGB) {bias=-0.5, scale=2}, B=ScalarParameter() {defaultValue=0.05, parameterName=InnerGlowNoiseAmount})) {b=true, r=true}
  n19 = SphereMask(
   A=Add(A=@n18, B=NamedRerouteUsage())
   B=Constant2Vector() {g=0.1, r=0.2}
   Radius=ScalarParameter() {defaultValue=0.3, parameterName=InnerGlowRadius}
   Hardness=ScalarParameter() {defaultValue=-0.5, parameterName=InnerGlowHardness}
  ) {attenuationRadius=12, hardnessPercent=55}
  n20 = Add(
   A=LinearInterpolate(
    A=LinearInterpolate(A=@n13, B=@n14.RGB, Alpha=@n19)
    B=VectorParameter() {parameterName=InnerScatteringColor, defaultValue=<rgba 0.2918 0.3564 0.9216 1>}.RGB
    Alpha=NamedRerouteUsage()
   )
   B=NamedRerouteUsage()
  )
  n21 = TextureCoordinate()
  n22 = Constant2Vector() {g=3, r=4}
  n23 = BumpOffset(
   Coordinate=Add(
    A=MaterialFunctionCall(UVs=@n21, Texture Scale=@n22) {materialFunction=&ScaleUVsByCenter}.UVs
    B=Constant2Vector() {g=-0.025, r=-0.135}
   )
   Height=ScalarParameter() {defaultValue=-0.1, parameterName=BumpOffsetLogoHeight}
  ) {heightRatio=0.1}
  n24 = VectorParameter() {defaultValue=<rgba 0.65 0.42 0.43 0.19>}
  n25 = MaterialFunctionCall!(
   UVs=@n23
   TextureIn=TextureObject() {samplerType=SAMPLERTYPE_LinearColor, texture=&/T_UpdootArrowGradient}
   UpperLeftCorner=AppendVector(A=@n24.A, B=@n24.A)
   LowerRightCorner=AppendVector(A=@n24.G, B=@n24.G)
   ExportFloat4
  ) {materialFunction=/Engine/Functions/Engine_MaterialFunctions01/Texturing/TextureCropping}
  n26 = Multiply(
   A=LinearInterpolate(A=@n20, B=Multiply(A=@n20) {constB=2}, Alpha=@n25.CroppedMasked)
   B=MaterialFunctionCall(
    Value_Morning=ScalarParameter() {defaultValue=1, parameterName=TOD_Morning}
    Value_Day=ScalarParameter() {defaultValue=1, parameterName=TOD_Day}
    Value_Evening=ScalarParameter() {defaultValue=1, parameterName=TOD_Evening}
    Value_Night=ScalarParameter() {defaultValue=0.6, parameterName=TOD_Night}
   ) {materialFunction=&Scalar_Time_of_Day}.Result
  )
  MaterialFunctionCall!(Color=@n26) {materialFunction=/Game/Packages/Fortress_SharedMaterials/Anime_Master_Material/MF_TODColorCurve}
  Multiply(A=@n25.CroppedMasked)
  n27 = LocalPosition() {includedOffsets=ExcludeOffsets, localOrigin=InstancePreSkinning}
  NamedRerouteDeclaration(None=ShaderStageSwitch(PixelShader=VertexInterpolator(VS=@n27.XYZ).PS, VertexShader=@n27.XYZ)) {name=PreskinnedPosition, nodeColor=<rgba 1 0.2344 0 1>}
  n28 = NamedRerouteUsage()
  n29 = VectorParameter() {parameterName=SSC, defaultValue=<rgba 0.8942 0.9624 1 1>}
  n30 = Saturate(None=DotProduct(
    A=AtmosphericLightVector()
    B=MaterialFunctionCall() {materialFunction=&CameraDirectionVector}.Result
   ))
  NamedRerouteDeclaration(None=LinearInterpolate(B=Multiply(A=@n28, B=@n29.RGB), Alpha=NamedRerouteDeclaration(None=@n30) {name=SunMask, nodeColor=<rgba 0.8047 0 1 1>})) {name=SSC, nodeColor=<rgba 0 0.4141 1 1>}
  NamedRerouteDeclaration(None=@n19) {name=InnerGlow, nodeColor=<rgba 0.1719 0 1 1>}
  n31 = Constant2Vector()
  n32 = ScreenPosition()
  n33 = MaterialFunctionCall(Float2=ViewSize()) {materialFunction=&BreakOutFloat2Components}
  n34 = AppendVector(A=Constant() {r=1}, B=Divide(A=@n33.G, B=@n33.G))
  n35 = Multiply(A=ConstantBiasScale(None=Add(A=@n31, B=@n32.ViewportUV)) {bias=-0.5, scale=2}, B=@n34)
  n36 = Multiply(
   A=Constant2Vector() {g=-0.5, r=0.5}
   B=MaterialFunctionCall(World Position=@n1) {materialFunction=&TransformToClipSpace}.Clip Space XY
  ) {constA=1}
  n37 = ScalarParameter() {defaultValue=6, parameterName=Screen Space Scale}
  n38 = MaterialFunctionCall() {materialFunction=&ObjectScale}
  n39 = MaterialFunctionCall(
   UVs=Subtract(A=ConstantBiasScale(None=@n35), B=Multiply(A=@n34, B=@n36))
   Texture Scale=Divide(A=Multiply(A=@n37, B=@n38.Scale X), B=ComponentMask(None=@n2) {b=true}) {constA=100}
  ) {materialFunction=&ScaleUVsByCenter}
  n40 = Constant2Vector() {g=0.3, r=0.4}
  NamedRerouteDeclaration(None=Multiply(A=Multiply(A=@n39.UVs, B=@n40), B=MaterialFunctionCall() {materialFunction=/Engine/Functions/Engine_MaterialFunctions02/Utility/FOV}."FOV ")) {name=ScreenUV, nodeColor=<rgba 1 0 0.1172 1>}
  n41 = NamedRerouteUsage()
  n42 = Constant()
  n43 = MaterialFunctionCall(
   Normal Vector=Add(A=AppendVector(A=@n41, B=@n42), B=VertexNormalWS())
   Power=ScalarParameter() {defaultValue=3, parameterName=OutGlowFresnelPow}
  ) {materialFunction=&Fresnel_Function}
  NamedRerouteDeclaration(None=Multiply(A=@n43.Result, B=VectorParameter() {parameterName=OutGlowFresnelColor, defaultValue=<rgba 0.4287 0.6514 0.3467 1>}.RGB)) {name=OuterGlow, nodeColor=<rgba 1 0.5156 0 1>}
  NamedRerouteDeclaration(None=@n18) {name=InnerGlowNoise, nodeColor=<rgba 0.3125 0 1 1>}
  NamedRerouteDeclaration(None=EyeAdaptationInverse(LightValueInput=@n26, AlphaInput=ScalarParameter() {defaultValue=1, parameterName=EyeAdaptionInverse}).EyeAdaptationInverse) {name=Emissive, nodeColor=<rgba 0.7344 1 0 1>}
  n44 = AtmosphericLightVector()
  n45 = PixelNormalWS()
  n46 = DotProduct(A=@n44, B=@n45)
  n47 = Subtract(A=Multiply(A=Multiply(A=@n46) {constB=2}, B=@n45), B=@n44)
  n48 = MaterialFunctionCall() {materialFunction=&CameraDirectionVector}
  n49 = Max(A=DotProduct(A=Normalize(VectorInput=@n47), B=Multiply(A=@n48.Result) {constB=-1})) {constB=0}
  n50 = ScalarParameter() {defaultValue=2, parameterName=InnerScatteringPower}
  NamedRerouteDeclaration(None=Multiply(A=Power(Base=@n49, Exp=@n50), B=ScalarParameter() {defaultValue=0.7, parameterName=InnerScatteringScale}) {constB=0.4}) {name=InnerScatter, nodeColor=<rgba 1 0 0.1406 1>}
  n51 = Constant3Vector() {constant=<rgba 0 0 1 0>}
  n52 = NamedRerouteUsage()
  n53 = Multiply(A=NamedRerouteUsage()) {constB=10}
  n54 = Add(A=@n52, B=Multiply(A=Saturate(None=@n53), B=ScalarParameter() {defaultValue=-0.01, parameterName=BubbleEdgeWarp}))
  n55 = BumpOffset(
   Coordinate=Multiply(A=Panner(Coordinate=@n54) {speedY=0.15}, B=ScalarParameter() {defaultValue=1.5, parameterName=BubbleTilingR}) {constB=3}
   Height=ScalarParameter() {defaultValue=0.05, parameterName=BubbleOffsetR}
  ) {heightRatio=0.1}
  n56 = TextureObjectParameter() {
   group=Slurp
   parameterName=BubbleMask
   samplerType=SAMPLERTYPE_Masks
   texture=/Game/Characters/Player/Male/Large/Bodies/M_LRG_Slurp_Monster/Textures/T_SlurpBubble
  }
  n57 = TextureSample(UVs=@n55, Tex=@n56)
  n58 = TextureObjectParameter() {
   group=Slurp
   parameterName=BubbleNormal
   samplerType=SAMPLERTYPE_Normal
   texture=/Game/Characters/Player/Male/Large/Bodies/M_LRG_Slurp_Monster/Textures/T_SlurpBubble_N
  }
  n59 = Multiply(A=Panner(Coordinate=@n52) {speedY=0.1}, B=ScalarParameter() {defaultValue=1, parameterName=BubbleTilingG}) {constB=3}
  n60 = TextureSample(UVs=@n59, Tex=@n56)
  n61 = Add(A=Multiply(A=@n57.R, B=TextureSample(UVs=@n55, Tex=@n58).RGB), B=Multiply(A=@n60.G, B=TextureSample(UVs=@n59, Tex=@n58).RGB))
  n62 = Add(A=@n57.R, B=@n60.G)
  n63 = Multiply(A=LinearInterpolate(Alpha=NamedRerouteUsage())) {constB=15}
  NamedRerouteDeclaration(None=MaterialFunctionCall(Normal=LinearInterpolate(A=@n51, B=@n61, Alpha=@n62), Flatness=Saturate(None=@n63)) {materialFunction=&FlattenNormal}.Result) {name=BubbleNormal, nodeColor=<rgba 0 1 0.7891 1>}
  NamedRerouteDeclaration(None=@n43.Result) {name=OuterFresnel, nodeColor=<rgba 0 0.1797 1 1>}
  NamedRerouteUsage()
  NamedRerouteUsage()
  OneMinus(None=@n21)
  n64 = PixelDepth()
  OneMinus(None=Saturate(None=Divide(A=@n64.R) {constB=100}))
  OneMinus(None=@n23)
  PixelNormalWS()
  n65 = Constant3Vector() {constant=<rgba 0.1572 0.9635 0.332 1>}
  n66 = VertexColor()
  SetMaterialAttributes!(
   MaterialAttributes
   Base Color=Constant()
   Metallic=Constant()
   Specular=Constant() {r=1}
   Roughness=Constant() {r=0.25}
   Normal=NamedRerouteUsage()
   Emissive Color=LinearInterpolate(A=NamedRerouteUsage(), B=@n65, Alpha=Add(A=@n66.R, B=@n66.R))
   Opacity=Constant() {r=1}
   Subsurface Color=LinearInterpolate(A=NamedRerouteUsage(), B=@n65, Alpha=@n66.G)
   World Position Offset=MaterialFunctionCall() {materialFunction=/SpecterGameplay/Gameplay/Effects/Updoot/MF_Sprite_Updoot_WPO}.Result
  ) {attributeSetTypes [
    #1
    #2
    #3
    #4
    #5
    #6
    #7
    #8
    #9
   ]}
 ]
}
